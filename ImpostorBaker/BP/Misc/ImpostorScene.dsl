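; ImpostorScene is the ImpostorBaker Blueprint that switches the capture scene between view modes for impostor baking.
; ChangeViewModes resets per-bake-type scalars in MPC_ImposterCapture, then either turns capture off for Lit, empty or CustomLighting, or enables the chosen mode's parameter.
; It also adjusts post-process settings and shows a decal for CustomLighting.
; SetUsePostProcess selects a blendable post-process material or a translucent sphere as the view-mode overlay; the other functions are empty.
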(fn ChangeViewModes
  (params (ViewMode :type Name))
  (bind _usepostprocessmaterial (Variables|Default|GetUsePostProcessMaterial))
  (bind _returnvalue (not _usepostprocessmaterial))
  (bind _curviewmode (LocalVariables|GetCurViewMode))
  (bind _returnvalue_2 (== _curviewmode "CustomLighting"))
  (bind _postprocess_viewmodes (Variables|Default|GetPostProcess_ViewModes))
  (bind _postprocess_viewmodes_2 (Variables|Default|GetPostProcess_ViewModes))
  (LocalVariables|SetCurViewMode ViewMode)
  (Utilities|FlowControl|Sequence
    (:then_0
      (bind _enumvalue (Utilities|Enum|ForEachEImpostorBakeTypes false)
        (:LoopBody
          (if (not (or (Utilities|Enum|Equal(Enum) _enumvalue "NewEnumerator9") (Utilities|Enum|Equal(Enum) _enumvalue "NewEnumerator18")))
            (Rendering|Material|SetScalarParameterValue "/ImpostorBaker/Materials/MPC/MPC_ImposterCapture.MPC_ImposterCapture" (Utilities|String|StringToName (Utilities|String|EnumtoString _enumvalue)) :!class /Script/Engine.KismetMaterialLibrary)))))
    (:then_1
      (if (Math|Boolean|ORBoolean (== (LocalVariables|GetCurViewMode) "Lit") (== (LocalVariables|GetCurViewMode) "") (== (LocalVariables|GetCurViewMode) "CustomLighting"))
        (Rendering|Material|SetScalarParameterValue "/ImpostorBaker/Materials/MPC/MPC_ImposterCapture.MPC_ImposterCapture" "EnableCapture" :!class /Script/Engine.KismetMaterialLibrary)
        (Rendering|SetVisibility (Variables|Default|GetViewMode_TranslucentSphere) _returnvalue false)
        (bind _structout (Utilities|Struct|SetmembersinPostProcessSettings (Variables|/Script/Engine.PostProcessComponent|GetSettings _postprocess_viewmodes_2) 1.0 1.0))
        (Variables|/Script/Engine.PostProcessComponent|SetSettings _structout _postprocess_viewmodes_2)
        (Rendering|SetVisibility (Variables|Default|GetDecal_CaptureCustomLighting) _returnvalue_2 false)
        (if _returnvalue_2
          (Utilities|Struct|SetmembersinPostProcessSettings (Variables|/Script/Engine.PostProcessComponent|GetSettings (Variables|Default|GetPostProcess_ViewModes)) 2.0 2.0))
        (else
          (Rendering|Material|SetScalarParameterValue "/ImpostorBaker/Materials/MPC/MPC_ImposterCapture.MPC_ImposterCapture" "EnableCapture" 1.0 :!class /Script/Engine.KismetMaterialLibrary)
          (Rendering|Material|SetScalarParameterValue "/ImpostorBaker/Materials/MPC/MPC_ImposterCapture.MPC_ImposterCapture" (LocalVariables|GetCurViewMode) 1.0 :!class /Script/Engine.KismetMaterialLibrary)
          (Rendering|SetVisibility (Variables|Default|GetViewMode_TranslucentSphere) _returnvalue false)
          (bind _structout_2 (Utilities|Struct|SetmembersinPostProcessSettings (Variables|/Script/Engine.PostProcessComponent|GetSettings _postprocess_viewmodes) 2.0 2.0))
          (Variables|/Script/Engine.PostProcessComponent|SetSettings _structout_2 _postprocess_viewmodes)
          (Rendering|SetVisibility (Variables|Default|GetDecal_CaptureCustomLighting) false false))))))

(fn ClearDynamicMesh)

(fn DisableCaptureMode)

(fn SetUsePostProcess
  (params (UsePostProcessMaterial :type Boolean))
  (Variables|Default|SetUsePostProcessMaterial UsePostProcessMaterial)
  (Utilities|Struct|SetmembersinPostProcessSettings (Variables|/Script/Engine.PostProcessComponent|GetSettings (Variables|Default|GetPostProcess_ViewModes)) (select (Variables|Default|GetUsePostProcessMaterial) (Utilities|Struct|MakeWeightedBlendables (Utilities|Array|MakeArray (Utilities|Struct|MakeWeightedBlendable 1.0 (Variables|Default|GetPostProcessMaterial)))) (Utilities|Struct|MakeWeightedBlendables)))
  (Rendering|SetVisibility (Variables|Default|GetViewMode_TranslucentSphere) (not (Variables|Default|GetUsePostProcessMaterial)) false))

(fn UpdateDynamicMesh)

(fn ConstructionScript)
