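; Impostor Baker helper that builds trimmed card meshes for baked impostors and billboards, reducing empty area around the silhouette.
; A material traces the alpha cutout into a tiny float render target, packing eight outline points into four pixels for editor-only CPU readback.
; Those points form a center-fanned card with atlas UVs, one per capture view for the multi-card billboard type.
; Geometry Script writes the result into a supplied DynamicMeshComponent with UVs, normals and tangents.
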(event EventBeginPlay)

(event EventTick
  (params (DeltaSeconds :type Float%20%28single-precision%29)))

(fn DrawDebugLines
  (Rendering|Debug|FlushPersistentDebugLines 0)
  (for _index (range (+ (Utilities|Array|LastIndex (Variables|CornerCutting|GetPointArray)) 1))
    (Rendering|Debug|DrawDebugLine 0 (Math|Conversions|ToVector(Vector2D) (* (- (Utilities|Array|Get(aref) (Variables|CornerCutting|GetPointArray) _index) 0.5) 128.0) 64.0) (Math|Conversions|ToVector(Vector2D) (* (- (Utilities|Array|Get(aref) (Variables|CornerCutting|GetPointArray) (Math|Integer|%25%28Integer%29 (+ _index 1) (Utilities|Array|Length (Variables|CornerCutting|GetPointArray)))) 0.5) 128.0) 64.0) "(R=1.000000,G=0.119992,B=0.362059,A=1.000000)" 60 2.0)))

(fn GenerateMeshData
  (bind (_r _g _b _a) (Utilities|Struct|BreakLinearColor (Utilities|Array|Get(aref) (Variables|Default|GetColorData) 3)))
  (bind (_r_2 _g_2 _b_2 _a_2) (Utilities|Struct|BreakLinearColor (Utilities|Array|Get(aref) (Variables|Default|GetColorData))))
  (bind (_r_3 _g_3 _b_3 _a_3) (Utilities|Struct|BreakLinearColor (Utilities|Array|Get(aref) (Variables|Default|GetColorData) 1)))
  (bind (_r_4 _g_4 _b_4 _a_4) (Utilities|Struct|BreakLinearColor (Utilities|Array|Get(aref) (Variables|Default|GetColorData) 2)))
  (bind (_x _y) (Default|DeriveAxes (LocalVariables|GetCurCardNormal)))
  (bind _vertidxoffset (LocalVariables|GetVertIdxOffset))
  (bind _vertidxoffset_2 (LocalVariables|GetVertIdxOffset))
  (Utilities|Array|Clear (Variables|Default|GetVerts))
  (Utilities|Array|Clear (Variables|Default|GetUVs))
  (Utilities|Array|Clear (Variables|Default|GetTris))
  (LocalVariables|SetVertIdxOffset 0)
  ;[1 1 1] Support both Billboards and Impostors
  (LocalVariables|SetCardNormalList (select (Utilities|Enum|Equal(Enum) (Variables|FromOwner|GetImpostor%20Type) "NewEnumerator2") (Variables|FromOwner|GetViewCaptureVectors) (Utilities|Array|MakeArray "0,0.000000,1.000000"))) ; Render card per view for Billboards
  (for (_array_element _array_index) (LocalVariables|GetCardNormalList)
    (LocalVariables|SetCurCardNormal _array_element)
    (LocalVariables|SetFrameIdx _array_index)
    ;[0.04167 0.04167 0.04167] Clear Per Card Arrays
    (Utilities|Array|Clear (Variables|CornerCutting|GetPointArray))
    (Utilities|Array|Clear (LocalVariables|GetTempVerts))
    ;[0.07087 0.3454 1] Use GPU to Trace Mesh Cutout
    (Utilities|ExtractSubFrame (LocalVariables|GetFrameIdx)
      (:Out
        (Rendering|ClearRenderTarget2D 0 (Variables|Default|GetMeshCutoutRT))
        (Rendering|DrawMaterialtoRenderTarget 0 (Variables|Default|GetMeshCutoutRT) (Variables|Default|GetMeshCutoutMID))
        ;; Sample Color Data to Point Array
        (Utilities|Array|Clear (Variables|Default|GetColorData))
        (Variables|Default|SetColorData (Rendering|RenderTargetSampleRectangleEditorOnly (Variables|Default|GetMeshCutoutRT) (Utilities|Struct|MakeLinearColor 0.0 0.0 2.0 2.0)))
        (Variables|CornerCutting|SetPointArray (Utilities|Array|MakeArray (Math|Vector2D|MakeVector2D 0.5 0.5) (Math|Vector2D|MakeVector2D _r_2 _g_2) (Math|Vector2D|MakeVector2D _b_2 _a_2) (Math|Vector2D|MakeVector2D _b_3 _a_3) (Math|Vector2D|MakeVector2D _r_3 _g_3) (Math|Vector2D|MakeVector2D _r _g) (Math|Vector2D|MakeVector2D _b _a) (Math|Vector2D|MakeVector2D _b_4 _a_4) (Math|Vector2D|MakeVector2D _r_4 _g_4)))
        (Utilities|FlowControl|Sequence
          (:then_0
            ;[1 0.7345 0.07855] Construct Vertex Array
            (for (_array_element_2 _array_index_2) (Variables|CornerCutting|GetPointArray)
              (LocalVariables|SetPointIdx _array_index_2)
              (Utilities|Array|Add (LocalVariables|GetTempVerts) (Utilities|Select :Index (Variables|FromOwner|GetImpostor%20Type) :NewEnumerator0 (+ (Math|Conversions|ToVector(Vector2D) (/ (* (Class|BPVectorFunctions|BiasScaleVector2D _array_element_2) (Variables|FromOwner|GetObject%20Radius)) 10)) (Variables|FromOwner|GetOffset%20Vector)) :NewEnumerator1 (+ (Math|Conversions|ToVector(Vector2D) (/ (* (Class|BPVectorFunctions|BiasScaleVector2D _array_element_2) (Variables|FromOwner|GetObject%20Radius)) 10)) (Variables|FromOwner|GetOffset%20Vector)) :NewEnumerator2 (+ (* (+ (* _x (Math|Vector2D|BreakVector2D (Class|BPVectorFunctions|BiasScaleVector2D _array_element_2))) (* _y (.Y (Math|Vector2D|BreakVector2D (Class|BPVectorFunctions|BiasScaleVector2D _array_element_2))))) (Variables|FromOwner|GetObject%20Radius)) (Variables|FromOwner|GetOffset%20Vector))))))
          (:then_1
            (Utilities|FlowControl|Sequence
              (:then_0
                ;[0.07087 0.3454 1] Construct UV Array
                (for _array_element_3 (Variables|CornerCutting|GetPointArray)
                  (Utilities|Array|Add (Variables|Default|GetUVs) (Utilities|Select :Index (Variables|FromOwner|GetImpostor%20Type) :NewEnumerator0 (/ (/ (* (+ (Class|BPVectorFunctions|SaturateVector2D _array_element_3) "(X=0.001000,Y=0.001000)") "(X=0.995000,Y=0.995000)") (Math|Conversions|ToFloat(Integer) (Variables|FromOwner|GetFramesXYInternal))) 10) :NewEnumerator1 (/ (/ (* (+ (Class|BPVectorFunctions|SaturateVector2D _array_element_3) "(X=0.001000,Y=0.001000)") "(X=0.995000,Y=0.995000)") (Math|Conversions|ToFloat(Integer) (Variables|FromOwner|GetFramesXYInternal))) 10) :NewEnumerator2 (+ (+ (/ (* (+ (Class|BPVectorFunctions|SaturateVector2D _array_element_3) "(X=0.001000,Y=0.001000)") "(X=0.995000,Y=0.995000)") (Math|Conversions|ToFloat(Integer) (Variables|FromOwner|GetFramesXYInternal))) (/ (Default|1dTo2dIndex (LocalVariables|GetFrameIdx) 3) 3)) (* (Math|Conversions|ToFloat(Boolean) (== (LocalVariables|GetFrameIdx) 4)) 2.0))))))
              (:then_1
                (Utilities|FlowControl|Sequence
                  (:then_0
                    ;[0.1983 1 0.6717] Build Index Buffer
                    (for _index (range (+ (- (Utilities|Array|LastIndex (LocalVariables|GetTempVerts)) 2) 1))
                      (Utilities|Array|Add (Variables|Default|GetTris) (+ (+ _index _vertidxoffset) 2)) ; This scheme connects every triangle back to the central vertex
                      (Utilities|Array|Add (Variables|Default|GetTris) (+ (+ _index _vertidxoffset) 1))
                      (Utilities|Array|Add (Variables|Default|GetTris) (LocalVariables|GetVertIdxOffset)))
                    (Utilities|Array|Add (Variables|Default|GetTris) (+ (Utilities|Array|LastIndex (LocalVariables|GetTempVerts)) (LocalVariables|GetVertIdxOffset)))
                    (Utilities|Array|Add (Variables|Default|GetTris) _vertidxoffset_2)
                    (Utilities|Array|Add (Variables|Default|GetTris) (+ _vertidxoffset_2 1)))
                  (:then_1
                    (Utilities|FlowControl|Sequence
                      (:then_0
                        ;[1 0.7345 0.07855] Finalize the Vertex Array
                        (for _array_element_4 (LocalVariables|GetTempVerts)
                          (Utilities|Array|Add (Variables|Default|GetVerts) _array_element_4))
                        (LocalVariables|SetVertIdxOffset (+ (Utilities|Array|LastIndex (Variables|Default|GetVerts)) 1)))))))))))) ; Only needed for Billboards
  )
  ;[0.1869 1 0.1612] Update the Dynamic Mesh
  (Default|UpdateDynamicMesh)
  (return (Variables|Default|GetVerts) (Variables|Default|GetTris) (Variables|Default|GetUVs)))

(fn Initialize
  (params (DynamicMeshComponent :type Dynamic%20Mesh%20Component%20Object%20Reference) (CombinedAlphas%20RT :type Texture%20Render%20Target%202D%20Object%20Reference) (Impostor%20Type :type EImposterLayoutType%20Enum :default "NewEnumerator0") (Object%20Radius :type Float%20%28double-precision%29) (Offset :type Vector) (FramesXY :type Integer) (Capture%20Vectors :type Array%20of%20Vectors))
  (Variables|Default|SetDynamicMeshComponent DynamicMeshComponent)
  (Variables|FromOwner|SetCombinedAlphasRT CombinedAlphas%20RT)
  (Variables|FromOwner|SetImpostor%20Type Impostor%20Type)
  (Variables|FromOwner|SetObject%20Radius Object%20Radius)
  (Variables|FromOwner|SetOffset%20Vector Offset)
  (Variables|FromOwner|SetFramesXYInternal FramesXY)
  (Variables|FromOwner|SetViewCaptureVectors Capture%20Vectors)
  (Rendering|CreateAndAssignMID (Variables|Default|GetMeshCutoutMID) "/ImpostorBaker/Materials/Generation/M_TraceMeshCutout.M_TraceMeshCutout"
    (:Out
      (Utilities|FlowControl|Sequence
        (:then_0
          (Utilities|IsValid (Variables|Default|GetMeshCutoutRT)
            (:Is%20Not%20Valid
              (bind _returnvalue (Rendering|CreateRenderTarget2D 0 2 2 "RTF_RGBA32f"))
              (Variables|Default|SetMeshCutoutRT _returnvalue))))
        (:then_1
          (Utilities|FlowControl|Sequence
            (:then_0
              (if (Utilities|Enum|Equal(Enum) (Variables|FromOwner|GetImpostor%20Type) "NewEnumerator2")
                (Utilities|IsValid (Variables|Default|GetFrameToTraceRT)
                  (:Is%20Valid
                    _)
                  (:Is%20Not%20Valid
                    (bind _returnvalue_2 (Rendering|CreateRenderTarget2D 0 16 16 "RTF_RGBA8"))
                    (Variables|Default|SetFrameToTraceRT _returnvalue_2)))
                (Rendering|CreateAndAssignMID (Variables|Default|GetExtractSubFrameMID) "/ImpostorBaker/Materials/Generation/M_ExtractSubframe_RT.M_ExtractSubframe_RT"
                  (:Out
                    (Rendering|Material|SetTextureParameterValue (Variables|Default|GetExtractSubFrameMID) "RT" (Variables|FromOwner|GetCombinedAlphasRT))))
                (else
                  (Variables|Default|SetFrameToTraceRT (Variables|FromOwner|GetCombinedAlphasRT)))))
            (:then_1
              (Rendering|Material|SetTextureParameterValue (Variables|Default|GetMeshCutoutMID) "RT" (Variables|Default|GetFrameToTraceRT))
              (Rendering|Material|SetScalarParameterValue (Variables|Default|GetMeshCutoutMID) "Enable Debug" 0.0 :!class /Script/Engine.MaterialInstanceDynamic))))))))

(fn SetMeshData
  (params (Verts :type Array%20of%20Vectors) (Tris :type Array%20of%20Integers) (UVs :type Array%20of%20Vector%202D%20Structures))
  (Variables|Default|SetVerts Verts)
  (Variables|Default|SetTris Tris)
  (Variables|Default|SetUVs UVs))

(fn UpdateDynamicMesh
  (bind _dynamicmesh (Variables|Components|GetDynamicMesh))
  (bind _returnvalue (DynamicMeshComponent|GetDynamicMesh (Variables|Default|GetDynamicMeshComponent)))
  (bind _output_get (Variables|Components|SetDynamicMesh _returnvalue))
  (DynamicMesh|Reset _output_get)
  ;[0.3886 0.3906 0.3866] Ensure Data Exists
  (if (> (Utilities|Array|Length (Variables|Default|GetVerts)) 0)
    (if (> (Utilities|Array|Length (Variables|Default|GetTris)) 0)
      (if (> (Utilities|Array|Length (Variables|Default|GetUVs)) 0)
        ;[1 0.8785 0.1476] Add Vertices
        (bind _vectorlist (GeometryScript|ListUtils|ConvertArraytoVectorList (Variables|Default|GetVerts)))
        (GeometryScript|MeshEdits|AddVerticestoMesh (Variables|Components|GetDynamicMesh) _vectorlist)
        (Utilities|FlowControl|Sequence
          (:then_0
            ;[0.3955 0.9801 1] Add Triangles
            (Utilities|Array|Clear (Variables|Default|GetDynamicMeshTris))
            (for _index (range (/ (Utilities|Array|Length (Variables|Default|GetTris)) 3))
              (bind _newtriangleindex (GeometryScript|MeshEdits|AddTriangletoMesh (Variables|Components|GetDynamicMesh) (Utilities|Struct|MakeIntVector (Utilities|Array|Get(acopy) (Variables|Default|GetTris) (* _index 3)) (Utilities|Array|Get(acopy) (Variables|Default|GetTris) (+ (* _index 3) 1)) (Utilities|Array|Get(acopy) (Variables|Default|GetTris) (+ (* _index 3) 2)))))
              (Utilities|Array|Add (Variables|Default|GetDynamicMeshTris) _newtriangleindex)))
          (:then_1
            ;[0.04988 0.1514 0.5885] Add UVs and Tangeents
            (GeometryScript|UVs|SetNumUVChannels (Variables|Components|GetDynamicMesh) 1)
            (for (_ _array_index) (Variables|Default|GetDynamicMeshTris)
              (LocalVariables|SetTriangleIdx _array_index)
              (GeometryScript|UVs|SetMeshTriangleUVs (Variables|Components|GetDynamicMesh) 0 (LocalVariables|GetTriangleIdx) (Utilities|Struct|MakeGeometryScriptUVTriangle (Utilities|Array|Get(acopy) (Variables|Default|GetUVs) (Utilities|Struct|BreakIntVector (.ReturnValue (GeometryScript|MeshQueries|GetTriangleIndices (Variables|Components|GetDynamicMesh) _array_index)))) (Utilities|Array|Get(acopy) (Variables|Default|GetUVs) (.Y (Utilities|Struct|BreakIntVector (.ReturnValue (GeometryScript|MeshQueries|GetTriangleIndices (Variables|Components|GetDynamicMesh) _array_index))))) (Utilities|Array|Get(acopy) (Variables|Default|GetUVs) (.Z (Utilities|Struct|BreakIntVector (.ReturnValue (GeometryScript|MeshQueries|GetTriangleIndices (Variables|Components|GetDynamicMesh) _array_index)))))))
              (GeometryScript|Normals|SetMeshTriangleNormals (Variables|Components|GetDynamicMesh) (LocalVariables|GetTriangleIdx) (Utilities|Struct|MakeGeometryScriptTriangle "0.000000,0.000000,1.000000" "0.000000,0.000000,1.000000" "0.000000,0.000000,1.000000")))
            (GeometryScript|Normals|SetMeshToPerVertexNormals(Computed) _dynamicmesh)
            (GeometryScript|Normals|ComputeTangents _dynamicmesh (Utilities|Struct|MakeGeometryScriptTangentsOptions "FastMikkT"))))
        (else
          (Development|PrintString 0 "Impostor_MeshCutout: Failed due to missing UV Data.")))
      (else
        (Development|PrintString 0 "Impostor_MeshCutout: Failed due to missing Triangle Data.")))
    (else
      (Development|PrintString 0 "Impostor_MeshCutout: Failed due to missing Vertex Data."))))
