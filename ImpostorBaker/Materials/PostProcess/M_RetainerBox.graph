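# Unlit UI-domain material in ImpostorBaker, named for a Retainer Box, that converts a texture's colour from sRGB to linear.
# The Texture parameter is sampled as linear colour, with a black 1x1 texture as its default.
# R, G and B each pass through sRGBToLinear and are recombined with MakeFloat3, leaving alpha unused.
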Material /ImpostorBaker/Materials/PostProcess/M_RetainerBox
graph {
 properties {materialDomain=MD_UI, pixelDepthOffsetMode=PDOM_Legacy, shadingModel=MSM_Unlit, usedShadingModels=Unlit}
 nodes [
  n1 = TextureSampleParameter2D() {
   parameterName=Texture
   samplerType=SAMPLERTYPE_LinearColor
   texture=/Engine/Functions/Engine_MaterialFunctions02/PivotPainter2/Black_1x1_EXR_Texture
  }
  MaterialFunctionCall(
   X=MaterialFunctionCall!(sRGB=@n1.R) {materialFunction=/Engine/Functions/Engine_MaterialFunctions01/ImageAdjustment/sRGBToLinear}.Linear
   Y=MaterialFunctionCall(sRGB=@n1.G) {materialFunction=&sRGBToLinear}.Linear
   Z=MaterialFunctionCall(sRGB=@n1.B) {materialFunction=&sRGBToLinear}.Linear
  ) {materialFunction=&MakeFloat3}
 ]
}
